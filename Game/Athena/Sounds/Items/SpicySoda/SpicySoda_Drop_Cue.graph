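# SpicySoda_Drop_Cue is the sound cue for dropping the Spicy Soda item in Athena.
# A single wave player plays the SpicySoda_Drop wave at reduced volume, with no randomization or layering.
# It uses shared weapon-drop attenuation and the gadgets-items sound class, and overrides concurrency to cap simultaneous copies by refusing new ones.
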SoundCue /Game/Athena/Sounds/Items/SpicySoda/SpicySoda_Drop_Cue
root=WavePlayer_0
nodes [{id=WavePlayer_0, properties {soundWaveAssetPtr=&/SpicySoda_Drop}}]
properties {
 attenuationSettings=&Fort_Weapon_Drop_Att
 bOverrideConcurrency=true
 maxDistance=3000
 soundClassObject=&Fort_Gadgets_Items
 volumeMultiplier=0.5
 concurrencyOverrides {maxCount=3, resolutionRule=PreventNew}
}
