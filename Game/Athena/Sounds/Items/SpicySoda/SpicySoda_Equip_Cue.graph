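# Sound cue for equipping the Spicy Soda item in Athena, playing the SpicySoda_Equip wave through a single wave player with no randomization or modulation.
# It routes to the player foley sound class and uses close-range prop attenuation.
# A concurrency override prevents new instances from starting once the limit is reached, rather than stopping ones already playing.
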SoundCue /Game/Athena/Sounds/Items/SpicySoda/SpicySoda_Equip_Cue
root=WavePlayer_1
nodes [{id=WavePlayer_1, properties {soundWaveAssetPtr=&/SpicySoda_Equip}}]
properties {
 attenuationSettings=/Game/Sounds/Attenuation/Fort_Prop_Close_Att
 bOverrideConcurrency=true
 maxDistance=1800
 soundClassObject=&Fort_Player_Foley
 volumeMultiplier=1
 concurrencyOverrides {resolutionRule=PreventNew}
}
